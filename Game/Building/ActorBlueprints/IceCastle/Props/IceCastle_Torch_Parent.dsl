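; Parent Blueprint for Ice Castle torch props, with tintable fire effects and torch audio that shut off when the prop is destroyed.
; Where cosmetic events can execute, the construction script puts a dynamic Ice King campfire material on the fire particle component and applies core, flame, ember and light colors from VFX variables.
; The death effects event stops the audio, while the replicated Destroyed notify deactivates the fire particles.
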(event EventOnDeathPlayEffects
  (params (Damage :type Float%20%28single-precision%29) (DamageTags :type Gameplay%20Tag%20Container%20Structure%20%28by%20ref%29) (Momentum :type Vector :default "0, 0, 0") (HitInfo :type Hit%20Result%20Structure%20%28by%20ref%29) (InstigatedBy :type Fort%20Pawn%20Object%20Reference) (DamageCauser :type Actor%20Object%20Reference) (EffectContext :type Gameplay%20Effect%20Context%20Handle%20Structure))
  (Audio|Components|Audio|Stop (Variables|Default|GetTorchAudio)))

(fn OnRep_Destroyed
  (if (Variables|Default|GetDestroyed)
    (Components|Activation|Deactivate (Variables|Default|GetP_Torch_Fire_01))))

(fn ConstructionScript
  (Networking|CanExecuteCosmeticEvents
    (:True
      (bind _returnvalue (Rendering|Material|CreateDynamicMaterialInstance (Variables|Default|GetP_Torch_Fire_01) 0 "/Game/Athena/Items/Traps/Campfire/Effects/MI_Fire_Campfire_Mesh_IceKing.MI_Fire_Campfire_Mesh_IceKing" "None" :!class /Script/Engine.PrimitiveComponent))
      (Rendering|Material|SetVectorParameterValue _returnvalue "Color_A" (Variables|VFX|GetCoreColor_A) :!class /Script/Engine.MaterialInstanceDynamic)
      (Rendering|Material|SetVectorParameterValue _returnvalue "Color_B" (Variables|VFX|GetCoreColor_B) :!class /Script/Engine.MaterialInstanceDynamic)
      (Effects|Components|ParticleSystem|SetVectorParameter (Variables|Default|GetP_Torch_Fire_01) "FlamesColor" (Math|Conversions|ToVector(LinearColor) (Variables|VFX|GetFlamesColor)))
      (Effects|Components|ParticleSystem|SetVectorParameter (Variables|Default|GetP_Torch_Fire_01) "EmbersColor" (Math|Conversions|ToVector(LinearColor) (Variables|VFX|GetEmbersColor)))
      (Effects|Components|ParticleSystem|SetVectorParameter (Variables|Default|GetP_Torch_Fire_01) "LightColor" (Math|Conversions|ToVector(LinearColor) (Variables|VFX|GetLightColor))))))
